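; BGA_FerrisWheel is a Ferris wheel building actor that is made invulnerable and can spin visually.
; On BeginPlay it calls the parent, then reads the EnableSpinWheelVisuals scalable float to switch the RotatingMovement component's tick on or off.
; When spinning is enabled, a second scalable float, HF_VisualsTickRate, sets that component's tick interval, so both the visual and its update cost are data-tunable.
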(event EventBeginPlay
  (bind _enablespinwheelvisuals (Variables|Default|GetEnableSpinWheelVisuals))
  (bind _returnvalue (ScalableFloat|GetValueasBool _enablespinwheelvisuals))
  ;[1 1 1] Init
  (|CallParentFunction)
  (Variables|BuildingActor|SetbIsInvulnerable true)
  (Components|Tick|SetComponentTickEnabled (Variables|Default|GetRotatingMovement) _returnvalue)
  (if _returnvalue
    (Components|Tick|SetComponentTickInterval (Variables|Default|GetRotatingMovement) (ScalableFloat|GetValueatLevel (Variables|Default|GetHF_VisualsTickRate)))))

(fn ConstructionScript)
