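; Invulnerable Ferris wheel building actor that lets players ride it as a zipline and repels pawns and vehicles from inside.
; Local interaction or client-side overlap requests ziplining on the nearer of two child loop actors.
; On authority, the inner volume launches sideways by position, pawn speed and a repeating timer, granting fall immunity to non-skydiving pawns.
; Hotfix values tune or disable features, and gameplay tags exempt actors.
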(event EventBeginPlay
  ;[1 1 1] Init
  (|CallParentFunction)
  (Variables|BuildingActor|SetbIsInvulnerable true)
  (if (or (ScalableFloat|GetValueasBool (Variables|HotFixValues|GetHF_DisableZipline)) (not (ScalableFloat|GetValueasBool (Variables|HotFixValues|GetHF_DoAutoZiplineOnOverlap))))
    (Variables|/Script/Engine.PrimitiveComponent|SetbGenerateOverlapEvents false (Variables|Default|GetFerrisWheelCollision))
    (Variables|Interaction|SetbAllowInteract false)
    (else
      _))
  (Variables|/Script/Engine.PrimitiveComponent|SetbGenerateOverlapEvents (ScalableFloat|GetValueasBool (Variables|HotFixValues|GetHF_EnablePushAwayVolume)) (Variables|Default|GetInsidePushAwayCollision)))

(event OnComponentBeginOverlap(FerrisWheelCollision)
  (params (OverlappedComponent :type Primitive%20Component%20Object%20Reference) (OtherActor :type Actor%20Object%20Reference) (OtherComp :type Primitive%20Component%20Object%20Reference) (OtherBodyIndex :type Integer) (bFromSweep :type Boolean) (SweepResult :type Hit%20Result%20Structure%20%28by%20ref%29))
  ;[1 1 1] Auto zipline overlap
  (CallFunction|OnZiplineOverlap OtherActor))

(event OnComponentBeginOverlap(InsidePushAwayCollision)
  (params (OverlappedComponent :type Primitive%20Component%20Object%20Reference) (OtherActor :type Actor%20Object%20Reference) (OtherComp :type Primitive%20Component%20Object%20Reference) (OtherBodyIndex :type Integer) (bFromSweep :type Boolean) (SweepResult :type Hit%20Result%20Structure%20%28by%20ref%29))
  ;[1 1 1] Push Away for inside of the wheel
  (Networking|SwitchHasAuthority
    (:Authority
      (CallFunction|OnPushawayOverlap OtherActor)
      (bind _asfort_player_pawn (Utilities|Casting|CastToFortPlayerPawn OtherActor))
      (Utilities|Set|Add (Variables|Default|GetPawnsInsidePushZone) _asfort_player_pawn)
      (bind _returnvalue (Utilities|Time|SetTimerbyEvent (EventDispatchers|CreateEvent :!event Custom|RepelInsidePawns) (ScalableFloat|GetValueatLevel (Variables|HotFixValues|GetHF_RepelTickRate)) true false (ScalableFloat|GetValueatLevel (Variables|HotFixValues|GetHF_RepelTickStartDelay))))
      (Variables|Default|SetRepelInsidePawnTimerHandle _returnvalue))))

(event OnComponentEndOverlap(InsidePushAwayCollision)
  (params (OverlappedComponent :type Primitive%20Component%20Object%20Reference) (OtherActor :type Actor%20Object%20Reference) (OtherComp :type Primitive%20Component%20Object%20Reference) (OtherBodyIndex :type Integer))
  (bind _pawnsinsidepushzone (Variables|Default|GetPawnsInsidePushZone))
  (Networking|SwitchHasAuthority
    (:Authority
      (CallFunction|OnPushawayEndOverlap OtherActor)
      (bind _asfort_player_pawn (Utilities|Casting|CastToFortPlayerPawn OtherActor))
      (Utilities|Set|Remove _pawnsinsidepushzone _asfort_player_pawn)
      (if (== (Utilities|Set|Length _pawnsinsidepushzone) 0)
        (Utilities|Time|ClearandInvalidateTimerbyHandle (Variables|Default|GetRepelInsidePawnTimerHandle))
        (else
          _)))))

(event Custom|RepelInsidePawns
  (Networking|SwitchHasAuthority
    (:Authority
      (bind _valuepin (Utilities|Set|ForEachLoop(Set) (Variables|Default|GetPawnsInsidePushZone)))
      (CallFunction|OnPushawayOverlap _valuepin))))

(fn BlueprintCanInteract
  (params (InteractingPawn :type Fort%20Pawn%20Object%20Reference) (InteractionBeingAttempted :type EInteractionBeingAttempted%20Enum :default "FirstInteraction") (InteractionType :type TInteractionType%20Enum :default "IT_NoInteraction"))
  (Utilities|IsValid InteractingPawn
    (:Is%20Valid
      (return (not (GameplayTags|HasAnyMatchingGameplayTags InteractingPawn (Variables|Default|GetTagsToAvoidForInsidePush)))))
    (:Is%20Not%20Valid
      (return false))))

(fn BlueprintGetInteractionString
  (params (InteractingPawn :type Fort%20Pawn%20Object%20Reference) (InteractionBeingAttempted :type EInteractionBeingAttempted%20Enum :default "FirstInteraction"))
  (return "NSLOCTEXT(\"[4F2F777777D04992062CEA621F5F3803]\", \"5501CFCF42E8B7D82D1270B44E8ED996\", \"Ride Ferris Wheel\")"))

(fn BlueprintOnLocalInteract
  (params (InteractingPawn :type Fort%20Player%20Pawn%20Object%20Reference) (InteractionBeingAttempted :type EInteractionBeingAttempted%20Enum :default "FirstInteraction"))
  (bind _returnvalue (Utilities|Casting|ToGameplayTagAssetInterface InteractingPawn))
  (Utilities|IsValid InteractingPawn
    (:Is%20Valid
      (if (Utilities|IsValidInterface _returnvalue)
        (if (not (GameplayTags|HasAnyMatchingGameplayTags _returnvalue (Variables|Default|GetTagsToAvoidForInsidePush)))
          (bind _returnvalue_2 (Zipline|IsZiplining InteractingPawn))
          (if (not (and _returnvalue_2 (Vehicle|IsinVehicle InteractingPawn)))
            (bind _asfort_movement_comp_character (Utilities|Casting|CastToFortMovementComp_Character (Variables|/Script/Engine.Character|GetCharacterMovement InteractingPawn))
              (:then
                (bind _asbp_ferris_wheel_zipline_loop (Utilities|Casting|CastToBP_FerrisWheel_Zipline_Loop (Utilities|SelectObject (Variables|/Script/Engine.ChildActorComponent|GetChildActor (Variables|Default|GetFerrisWheel_Zipline_side2)) (Variables|/Script/Engine.ChildActorComponent|GetChildActor (Variables|Default|GetFerrisWheel_Zipline_side1)) (> (Transformation|GetDistanceTo InteractingPawn (Variables|/Script/Engine.ChildActorComponent|GetChildActor (Variables|Default|GetFerrisWheel_Zipline_side1))) (Transformation|GetDistanceTo InteractingPawn (Variables|/Script/Engine.ChildActorComponent|GetChildActor (Variables|Default|GetFerrisWheel_Zipline_side2))))))
                  (:then
                    (bind _returnvalue_3 (Class|FortMovementExternalZiplineProvider|RequestZiplining (Class|FortMovementCompCharacter|GetExternalZipliningProvider _asfort_movement_comp_character) _asbp_ferris_wheel_zipline_loop (Variables|/Script/Engine.Character|GetMesh InteractingPawn) 0 "Forward"))
                    (return _returnvalue_3))
                  (:CastFailed
                    _)))
              (:CastFailed
                _))
            (return false)))
        (else
          _))
      (return false))
    (:Is%20Not%20Valid
      _)))

(fn GetPawnOverlapBehavior
  (params (Pawn :type Fort%20Player%20Pawn%20Object%20Reference))
  (bind _charactermovement (Variables|/Script/Engine.Character|GetCharacterMovement Pawn))
  (bind _returnvalue (Pawn|Components|CharacterMovement|GetLastUpdateVelocity _charactermovement))
  (bind _returnvalue_2 (Math|Vector|VectorLength _returnvalue))
  (Development|PrintString (Utilities|String|ToString(Text) (Utilities|Text|FormatText "NSLOCTEXT(\"[4F2F777777D04992062CEA621F5F3803]\", \"2A508D8E40691CD443CEE8AC1347786A\", \"Velocity of overlapping pawn {vel}\")" _returnvalue_2)))
  (if (>= _returnvalue_2 (ScalableFloat|GetValueatLevel (Variables|HotFixValues|GetHF_RepelVelocityThreshold)))
    (return "NewEnumerator1")
    (else
      (return "NewEnumerator0"))))

(fn GetRepelDirection
  (params (Behavior :type FerrisWheelOverlapBehavior%20Enum :default "NewEnumerator0") (DotProduct :type Float%20%28double-precision%29))
  (bind _returnvalue (>= DotProduct 0))
  (switch Utilities|FlowControl|Switch|SwitchonFerrisWheelOverlapBehavior Behavior
    (:NewEnumerator0
      (return (Math|Vector|SelectVector (Transformation|GetActorRightVector) (* (Transformation|GetActorRightVector) -1.0) _returnvalue)))
    (:NewEnumerator1
      (return (Math|Vector|SelectVector (Transformation|GetActorRightVector) (* (Transformation|GetActorRightVector) -1.0) (not _returnvalue))))
    (:NewEnumerator2
      (return "0, 0, 0"))))

(fn OnPushawayEndOverlap
  (params (EndOverlapActor :type Actor%20Object%20Reference)))

(fn OnPushawayOverlap
  (params (OverlappedActor :type Actor%20Object%20Reference))
  (bind _overlappingpawn (LocalVariables|GetOverlappingPawn))
  (bind _repel (CallFunction|GetPawnOverlapBehavior _overlappingpawn))
  (bind _overlappingpawn_2 (LocalVariables|GetOverlappingPawn))
  (bind _returnvalue (Ability|GetAbilitySystemComponent _overlappingpawn_2))
  (Utilities|IsValid OverlappedActor
    (:Is%20Valid
      (bind _asgameplay_tag_asset_interface (Utilities|Casting|CastToGameplayTagAssetInterface (Variables|GetOverlappedActor))
        (:then
          (if (GameplayTags|HasAnyMatchingGameplayTags _asgameplay_tag_asset_interface (Variables|Default|GetTagsToAvoidForInsidePush))
            (return)
            (else
              (LocalVariables|SetActorForwardDotRight (Math|Vector|DotProduct (Transformation|GetActorRightVector) (- (Transformation|GetActorLocation (Variables|GetOverlappedActor)) (Transformation|GetActorLocation))))
              (bind _asfort_player_pawn (Utilities|Casting|CastToFortPlayerPawn (Variables|GetOverlappedActor))
                (:then
                  ;[1 1 1] Pawns
                  (LocalVariables|SetOverlappingPawn _asfort_player_pawn)
                  (bind _returnvalue_2 (Zipline|IsZiplining (LocalVariables|GetOverlappingPawn)))
                  (if _returnvalue_2
                    (return)
                    (else
                      (switch Utilities|FlowControl|Switch|SwitchonFerrisWheelOverlapBehavior _repel
                        (:NewEnumerator0
                          _)
                        (:NewEnumerator1
                          _))
                      (Character|LaunchCharacter (LocalVariables|GetOverlappingPawn) (* (CallFunction|GetRepelDirection _repel (LocalVariables|GetActorForwardDotRight)) (ScalableFloat|GetValueatLevel (Variables|HotFixValues|GetHF_RepelVelocity_Pawn))) false false)
                      (if (Parachute|IsSkydiving (LocalVariables|GetOverlappingPawn))
                        (AthenaMovement|SetPendingSkydiveLaunch (LocalVariables|GetOverlappingPawn) true)
                        (else
                          (Utilities|IsValid _returnvalue
                            (:Is%20Valid
                              (GameplayEffects|ApplyGameplayEffectToSelf :self _returnvalue :GameplayEffectClass "/Game/Athena/Items/Parents/FallImmunity/GE_Athena_TillLandFallDamageImmunity_Grant_Parent.GE_Athena_TillLandFallDamageImmunity_Grant_Parent_C" :Level 0.0))))))))
                (:CastFailed
                  ;[1 1 1] Vehicles
                  (bind _asfort_athena_vehicle (Utilities|Casting|CastToFortAthenaVehicle (Variables|GetOverlappedActor)))
                  (Class|VehicleFunctionLibrary|LaunchVehicle _asfort_athena_vehicle (* (CallFunction|GetRepelDirection "NewEnumerator1" (LocalVariables|GetActorForwardDotRight)) (ScalableFloat|GetValueatLevel (Variables|HotFixValues|GetHF_RepelVelocity_Vehicles)))))))))
        (:CastFailed
          (return))))
    (:Is%20Not%20Valid
      (return))))

(fn OnZiplineOverlap
  (params (OverlappedActor :type Actor%20Object%20Reference))
  (Utilities|IsValid OverlappedActor
    (:Is%20Valid
      (Networking|SwitchHasAuthority
        (:Remote
          (bind _asgameplay_tag_asset_interface (Utilities|Casting|CastToGameplayTagAssetInterface OverlappedActor))
          (if (not (GameplayTags|HasAnyMatchingGameplayTags _asgameplay_tag_asset_interface (Variables|Default|GetTagsToAvoidForOverlap)))
            (bind _asfort_player_pawn (Utilities|Casting|CastToFortPlayerPawn OverlappedActor))
            (bind _returnvalue (Zipline|IsZiplining _asfort_player_pawn))
            (if (and (not _returnvalue) (not (Vehicle|IsinVehicle _asfort_player_pawn)))
              (bind _asfort_movement_comp_character (Utilities|Casting|CastToFortMovementComp_Character (Variables|/Script/Engine.Character|GetCharacterMovement _asfort_player_pawn)))
              (bind _asbp_ferris_wheel_zipline_loop (Utilities|Casting|CastToBP_FerrisWheel_Zipline_Loop (Utilities|SelectObject (Variables|/Script/Engine.ChildActorComponent|GetChildActor (Variables|Default|GetFerrisWheel_Zipline_side2)) (Variables|/Script/Engine.ChildActorComponent|GetChildActor (Variables|Default|GetFerrisWheel_Zipline_side1)) (> (Transformation|GetDistanceTo OverlappedActor (Variables|/Script/Engine.ChildActorComponent|GetChildActor (Variables|Default|GetFerrisWheel_Zipline_side1))) (Transformation|GetDistanceTo OverlappedActor (Variables|/Script/Engine.ChildActorComponent|GetChildActor (Variables|Default|GetFerrisWheel_Zipline_side2)))))))
              (Class|FortMovementExternalZiplineProvider|RequestZiplining (Class|FortMovementCompCharacter|GetExternalZipliningProvider _asfort_movement_comp_character) _asbp_ferris_wheel_zipline_loop (Variables|/Script/Engine.Character|GetMesh _asfort_player_pawn) 0 "Forward"))))))))

(fn ConstructionScript)
