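; Spline zipline Blueprint that serves as the "Ride Ferris Wheel" interaction, switchable off by a scalable float.
; Mounting applies an on-wheel GameplayEffect, broadcasts a custom verb and plays local audio, with socket offset and velocity direction overrides.
; Dismounting swaps in a launch effect and, when launching is enabled, launches the pawn forward and up, then triggers skydive unless tag-blocked.
; Construction script derives spline endpoints from two pole sockets, auto-sets points and updates interaction collision.
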(event EventBeginPlay
  (|CallParentFunction)
  (Variables|Interaction|SetbAllowInteract (not (ScalableFloat|GetValueasBool (Variables|Default|GetDisableZiplining)))))

(event EventOnZipliningStopped
  (params (InteractingPawn :type Fort%20Player%20Pawn%20Object%20Reference) (InteractComponent :type Primitive%20Component%20Object%20Reference) (ZiplineUsageDuration :type Float%20%28single-precision%29))
  (bind _returnvalue (Ability|GetAbilitySystemComponent InteractingPawn))
  (bind (_returnvalue_x _returnvalue_y) (Math|Vector|BreakVector (* (Transformation|GetActorForwardVector InteractingPawn) (ScalableFloat|GetValueatLevel (Variables|Default|GetLaunchXYMultiplier)))))
  (Utilities|IsValid InteractingPawn
    (:Is%20Valid
      (Utilities|IsValid _returnvalue
        (:Is%20Valid
          (GameplayEffects|RemoveActiveEffectswithGrantedTags _returnvalue "(GameplayTags=((TagName=\"FerrisWheel.OnWheel\")))")
          (GameplayEffects|ApplyGameplayEffectToSelf :self _returnvalue :GameplayEffectClass (Variables|Default|GetFerrisWheelLaunchEffect) :Level 0.0)
          (if (ScalableFloat|GetValueasBool (Variables|Default|GetLaunchingEnabled))
            (Character|LaunchCharacter InteractingPawn (Math|Vector|MakeVector _returnvalue_x _returnvalue_y (ScalableFloat|GetValueatLevel (Variables|Default|GetLaunchZValue))) false false)
            (CallFunction|TrySkyDive InteractingPawn)
            (Default|CallOnPlayerEndZiplining InteractingPawn)
            (if (Pawn|IsLocallyViewed InteractingPawn)
              (Audio|SpawnSound2D "/FerrisWheelGameplay/Audio/MSS_FerrisWheel_Dismount.MSS_FerrisWheel_Dismount")
              (Development|PrintString "[FerrisWheel] Player ending zipline"))))))))

(event EventOnZipliningStarted
  (params (InteractingPawn :type Fort%20Player%20Pawn%20Object%20Reference) (InteractComponent :type Primitive%20Component%20Object%20Reference))
  (bind _returnvalue (Ability|GetAbilitySystemComponent InteractingPawn))
  (Utilities|IsValid InteractingPawn
    (:Is%20Valid
      (Utilities|IsValid _returnvalue
        (:Is%20Valid
          (GameplayEffects|ApplyGameplayEffectToSelf :self _returnvalue :GameplayEffectClass (Variables|Default|GetFerrisWheelOnEffect) :Level 0.0)
          (CustomVerb|BroadcastCustomVerbMessage (CustomVerb|MakeCustomVerbMessage InteractingPawn "(TagName=\"CustomVerb.ThemePark.FerrisWheel.AttachedPlayer\")"))
          (Default|CallOnPlayerBeginZiplining InteractingPawn)
          (if (Pawn|IsLocallyViewed InteractingPawn)
            (Audio|SpawnSound2D "/FerrisWheelGameplay/Audio/MSS_FerrisWheel_Mount.MSS_FerrisWheel_Mount")
            (Development|PrintString "[FerrisWheel] Player starting zipline")))))))

(fn AttemptOverrideZiplineSocketOffset
  (params (InteractingPawn :type Fort%20Player%20Pawn%20Object%20Reference) (InteractComponent :type Primitive%20Component%20Object%20Reference) (BaseSocketOffset :type Vector%20%28by%20ref%29 :default "0, 0, 0"))
  (return (Variables|Default|GetSocketOffsetOverride) true))

(fn BlueprintGetInteractionString
  (params (InteractingPawn :type Fort%20Pawn%20Object%20Reference) (InteractionBeingAttempted :type EInteractionBeingAttempted%20Enum :default "FirstInteraction"))
  (return "NSLOCTEXT(\"[28B929215584C303D545465543147996]\", \"848B53164B2B1BCA93BD41875BDE7546\", \"Ride Ferris Wheel\")"))

(fn ShouldOverrideVelocityDirection
  (params (InteractingPawn :type Fort%20Player%20Pawn%20Object%20Reference) (InteractComponent :type Primitive%20Component%20Object%20Reference))
  (return (Variables|Default|GetVelocityDirectionOverride) (not (Math|Float|NearlyEqual(Float) (Variables|Default|GetVelocityDirectionOverride)))))

(fn SkydivePlayer
  (params (FortPlayer :type Fort%20Player%20Pawn%20Object%20Reference))
  (bind _output_get (LocalVariables|SetPlayerRef FortPlayer))
  (AthenaMovement|SetPendingSkydiveLaunch _output_get true))

(fn TrySkyDive
  (params (Pawn :type Fort%20Player%20Pawn%20Object%20Reference))
  (if (not (GameplayTags|HasAnyMatchingGameplayTags Pawn (Variables|Default|GetTags_Blocked_SkyDive)))
    (AthenaMovement|SetPendingSkydiveLaunch Pawn true)))

(fn ConstructionScript
  (bind _ziplinespline (Variables|FortAthenaSplineZipline|GetZiplineSpline))
  (bind _returnvalue (Spline|GetNumberOfSplinePoints _ziplinespline))
  (bind _returnvalue_2 (- _returnvalue 1))
  (bind (_poleasocketlocation_x _poleasocketlocation_y _poleasocketlocation_z) (Math|Vector|BreakVector (Variables|GetPoleASocketLocation_X)))
  (bind (_polebsocketlocation_x _polebsocketlocation_y _polebsocketlocation_z) (Math|Vector|BreakVector (Variables|GetPoleBSocketLocation_X)))
  (bind _polea (Variables|Default|GetPoleA))
  (bind _poleb (Variables|Default|GetPoleB))
  (Utilities|IsValid _polea
    (:Is%20Valid
      (Variables|Default|SetPoleASocketLocation (Transformation|GetSocketLocation (Variables|/Script/Engine.Actor|GetRootComponent _polea) "LineAttachSocket"))
      (Utilities|IsValid _poleb
        (:Is%20Valid
          (Variables|Default|SetPoleBSocketLocation (Transformation|GetSocketLocation (Variables|/Script/Engine.Actor|GetRootComponent _poleb) "LineAttachSocket"))
          (if (< _poleasocketlocation_z _polebsocketlocation_z)
            (Variables|Default|SetLowerPointID 0)
            (Variables|Default|SetHigherPointID (- (Spline|GetNumberOfSplinePoints (Variables|FortAthenaSplineZipline|GetZiplineSpline)) 1))
            (Variables|Default|SetHigherEndLocation (Variables|Default|GetPoleASocketLocation))
            (Variables|Default|SetLowerEndLocation (Variables|Default|GetPoleBSocketLocation))
            (else
              (Variables|Default|SetHigherPointID 0)
              (Variables|Default|SetLowerPointID (- (Spline|GetNumberOfSplinePoints (Variables|FortAthenaSplineZipline|GetZiplineSpline)) 1))
              (Variables|Default|SetLowerEndLocation (Variables|Default|GetPoleASocketLocation))
              (Variables|Default|SetHigherEndLocation (Variables|Default|GetPoleBSocketLocation))))
          (for _index (range (+ _returnvalue_2 1))
            (if (or (== _index 0) (== _index _returnvalue_2))
              (CallFunction|SetSplinePositionandTangent (Variables|Default|GetAuto%20Set%20Spline%20Ends) (Variables|Default|GetAuto%20Set%20Spline%20Ends) _index)
              (else
                (CallFunction|SetSplinePositionandTangent (Variables|Default|GetAuto%20Set%20Spline%20Mids) (Variables|Default|GetAuto%20Set%20Spline%20Mids) _index))))
          (Spline|UpdateSpline (Variables|FortAthenaSplineZipline|GetZiplineSpline)))))
    (:Is%20Not%20Valid
      _))
  (CallFunction|SetInteractionCollision))
